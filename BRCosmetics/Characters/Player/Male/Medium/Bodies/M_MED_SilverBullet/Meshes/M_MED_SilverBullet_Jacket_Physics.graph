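# Physics asset for the SilverBullet medium male body's jacket, setting up seven jacket panels to swing around the character.
# Capsules on the spine, pelvis, thighs, clavicles, arms and hands form the body.
# Each panel has a small base body locked to spine_03 and a dyn body with tight angular limits, narrowed to one swing axis at the back.
# Extra free-angle constraints tie the back-left and back-right panels to the centre back.
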PhysicsAsset /BRCosmetics/Characters/Player/Male/Medium/Bodies/M_MED_SilverBullet/Meshes/M_MED_SilverBullet_Jacket_Physics
bodies {
 base_jacket_bck [{shapeType=Capsule, length=0.505, radius=0.505, shapeName=base_jacket_bck_capsule, rotation=<rot 0 -90 0>}]
 base_jacket_bck_l [{shapeType=Capsule, length=0.505, radius=0.505, shapeName=base_jacket_bck_l_capsule, rotation=<rot 0 -90 0>}]
 base_jacket_bck_r [{shapeType=Capsule, length=0.505, radius=0.505, shapeName=base_jacket_bck_r_capsule, rotation=<rot 0 -90 0>}]
 base_jacket_fr_l [{shapeType=Capsule, length=0.005, radius=0.5, shapeName=base_jacket_fr_l_capsule, rotation=<rot 0 -90 0>}]
 base_jacket_fr_r [{shapeType=Capsule, length=0.505, radius=0.505, shapeName=base_jacket_fr_r_capsule, rotation=<rot 0 -90 0>}]
 base_jacket_l [{shapeType=Capsule, length=0.505, radius=0.505, shapeName=base_jacket_l_capsule, rotation=<rot 0 -90 0>}]
 base_jacket_r [{shapeType=Capsule, length=0.505, radius=0.505, shapeName=base_jacket_r_capsule, rotation=<rot 0 -90 0>}]
 clavicle_l [{shapeType=Capsule, length=19.3997, radius=10.2151, center=<20.2817 0.4714 -7.3516>, rotation=<rot 0 -84.7943 -48.1631>}]
 clavicle_r [{shapeType=Capsule, length=19.4946, radius=10.484, center=<-21.0917 -0.7508 6.2812>, rotation=<rot 0 -85.8824 -51.9927>}]
 dyn_jacket_bck [{
   shapeType=Capsule
   length=7.7395
   radius=3.2477
   shapeName=dyn_jacket_bck_capsule
   center=<6.7732 0 2.3424>
   rotation=<rot 82.9583 -180 -180>
  }]
 dyn_jacket_bck_l [{
   shapeType=Capsule
   length=10.6826
   radius=4.032
   shapeName=dyn_jacket_bck_l_capsule
   center=<9.1808 0 2.961>
   rotation=<rot 84.162 -179.9999 180>
  }]
 dyn_jacket_bck_r [{
   shapeType=Capsule
   length=10.6826
   radius=4.032
   shapeName=dyn_jacket_bck_r_capsule
   center=<-9.1808 0 -2.961>
   rotation=<rot 84.162 -179.9999 0>
  }]
 dyn_jacket_fr_l [{
   shapeType=Capsule
   length=12.4084
   radius=6.1141
   shapeName=dyn_jacket_fr_l_capsule
   center=<7 0 4.5>
   rotation=<rot 0.7235 80.9036 -85.0613>
  }]
 dyn_jacket_fr_r [{
   shapeType=Capsule
   length=12.4084
   radius=6.1141
   shapeName=dyn_jacket_fr_r_capsule
   center=<-7 0 -4.5>
   rotation=<rot 0.7235 80.9036 94.9387>
  }]
 dyn_jacket_l [{
   shapeType=Capsule
   length=9.6629
   radius=3.7592
   shapeName=dyn_jacket_l_capsule
   center=<8.9273 0 2.7451>
   rotation=<rot 80.8424 -179.9999 180>
  }]
 dyn_jacket_r [{
   shapeType=Capsule
   length=9.6629
   radius=3.7592
   shapeName=dyn_jacket_r_capsule
   center=<-8.9273 0 -2.7451>
   rotation=<rot 80.8424 -179.9999 0>
  }]
 hand_l [{shapeType=Capsule, length=9.3971, radius=5.7859, center=<7.7463 -2.0502 -0.0994>, rotation=<rot 0 81.1612 -79.7944>}]
 hand_r [{shapeType=Capsule, length=9.3899, radius=5.8417, center=<-7.6828 2.0504 0.0223>, rotation=<rot 0 81.7818 -78.4464>}]
 lowerarm_l [{shapeType=Capsule, length=14.2063, radius=6.7441, center=<9.5274 0.8185 0.2323>, rotation=<rot 0 -84.2656 -79.5706>}]
 lowerarm_r [{shapeType=Capsule, length=14.7453, radius=6.6168, center=<-10.1654 -0.8842 -0.43>, rotation=<rot 0 -83.835 -81.3049>}]
 pelvis [{shapeType=Capsule, length=5.0947, radius=13.8098}]
 spine_01 [{shapeType=Capsule, length=14.5752, radius=8.2392, shapeName=spine_01_capsule, rotation=<rot 0 -90 0>}]
 spine_02 [{shapeType=Capsule, length=8.1519, radius=10.4543, shapeName=spine_02_capsule}]
 spine_03 [{shapeType=Capsule, length=9.8405, radius=9.1709}]
 spine_04 [{shapeType=Capsule, length=10, radius=1}]
 spine_05 [{shapeType=Capsule, length=10, radius=1}]
 thigh_l [{shapeType=Capsule, length=23.7212, radius=9.3252, center=<-16.0158 0 0>, rotation=<rot 90 0 0>}]
 thigh_r [{shapeType=Capsule, length=23.7212, radius=9.3252, center=<16.0158 0 0>, rotation=<rot 90 0 180>}]
 upperarm_twist_01_l [{shapeType=Capsule, length=9.0185, radius=8.4339, center=<4.0781 -0.2935 -2.2382>, rotation=<rot 2.2858 -92.8099 -90>}]
 upperarm_twist_01_r [{shapeType=Capsule, length=9.2291, radius=8.5854, center=<-3.8852 0.1118 2.1486>, rotation=<rot 2.276 -92.3209 -90>}]
}
constraints [
 {bone1Name=spine_03, bone2Name=spine_02, swing1LimitDegrees=45, swing2LimitDegrees=45, twistLimitDegrees=45}
 {bone1Name=spine_04, bone2Name=spine_03, swing1LimitDegrees=45, swing2LimitDegrees=45, twistLimitDegrees=45}
 {bone1Name=spine_05, bone2Name=spine_04, swing1LimitDegrees=45, swing2LimitDegrees=45, twistLimitDegrees=45}
 {bone1Name=clavicle_l, bone2Name=spine_05, swing1LimitDegrees=45, swing2LimitDegrees=45, twistLimitDegrees=45}
 {bone1Name=lowerarm_l, bone2Name=clavicle_l, swing1LimitDegrees=45, swing2LimitDegrees=45, twistLimitDegrees=45}
 {bone1Name=hand_l, bone2Name=lowerarm_l, swing1LimitDegrees=45, swing2LimitDegrees=45, twistLimitDegrees=45}
 {bone1Name=upperarm_twist_01_l, bone2Name=clavicle_l, swing1LimitDegrees=45, swing2LimitDegrees=45, twistLimitDegrees=45}
 {bone1Name=clavicle_r, bone2Name=spine_05, swing1LimitDegrees=45, swing2LimitDegrees=45, twistLimitDegrees=45}
 {bone1Name=lowerarm_r, bone2Name=clavicle_r, swing1LimitDegrees=45, swing2LimitDegrees=45, twistLimitDegrees=45}
 {bone1Name=hand_r, bone2Name=lowerarm_r, swing1LimitDegrees=45, swing2LimitDegrees=45, twistLimitDegrees=45}
 {bone1Name=upperarm_twist_01_r, bone2Name=clavicle_r, swing1LimitDegrees=45, swing2LimitDegrees=45, twistLimitDegrees=45}
 {bone1Name=thigh_l, bone2Name=pelvis, swing1LimitDegrees=45, swing2LimitDegrees=45, twistLimitDegrees=45}
 {bone1Name=thigh_r, bone2Name=pelvis, swing1LimitDegrees=45, swing2LimitDegrees=45, twistLimitDegrees=45}
 {
  bone1Name=spine_02
  bone2Name=spine_01
  swing1LimitDegrees=45
  swing1Motion=Free
  swing2LimitDegrees=45
  swing2Motion=Free
  twistLimitDegrees=45
  twistMotion=Free
 }
 {
  bone1Name=spine_01
  bone2Name=pelvis
  swing1LimitDegrees=45
  swing1Motion=Free
  swing2LimitDegrees=45
  swing2Motion=Free
  twistLimitDegrees=45
  twistMotion=Free
 }
 {
  bone1Name=base_jacket_bck
  bone2Name=spine_03
  swing1LimitDegrees=45
  swing1Motion=Locked
  swing2LimitDegrees=45
  swing2Motion=Locked
  twistLimitDegrees=45
  twistMotion=Locked
 }
 {bone1Name=dyn_jacket_bck, bone2Name=base_jacket_bck, swing1LimitDegrees=0, swing2LimitDegrees=20, twistLimitDegrees=0}
 {
  bone1Name=base_jacket_bck_l
  bone2Name=spine_03
  swing1LimitDegrees=45
  swing1Motion=Locked
  swing2LimitDegrees=45
  swing2Motion=Locked
  twistLimitDegrees=45
  twistMotion=Locked
 }
 {bone1Name=dyn_jacket_bck_l, bone2Name=base_jacket_bck_l, swing1LimitDegrees=0, swing2LimitDegrees=15, twistLimitDegrees=0}
 {
  bone1Name=base_jacket_bck_r
  bone2Name=spine_03
  swing1LimitDegrees=45
  swing1Motion=Locked
  swing2LimitDegrees=45
  swing2Motion=Locked
  twistLimitDegrees=45
  twistMotion=Locked
 }
 {bone1Name=dyn_jacket_bck_r, bone2Name=base_jacket_bck_r, swing1LimitDegrees=0, swing2LimitDegrees=15, twistLimitDegrees=0}
 {
  bone1Name=base_jacket_fr_l
  bone2Name=spine_03
  swing1LimitDegrees=45
  swing1Motion=Locked
  swing2LimitDegrees=45
  swing2Motion=Locked
  twistLimitDegrees=45
  twistMotion=Locked
 }
 {bone1Name=dyn_jacket_fr_l, bone2Name=base_jacket_fr_l, swing1LimitDegrees=10, swing2LimitDegrees=20, twistLimitDegrees=10}
 {
  bone1Name=base_jacket_fr_r
  bone2Name=spine_03
  swing1LimitDegrees=45
  swing1Motion=Locked
  swing2LimitDegrees=45
  swing2Motion=Locked
  twistLimitDegrees=45
  twistMotion=Locked
 }
 {bone1Name=dyn_jacket_fr_r, bone2Name=base_jacket_fr_r, swing1LimitDegrees=10, swing2LimitDegrees=20, twistLimitDegrees=10}
 {
  bone1Name=base_jacket_l
  bone2Name=spine_03
  swing1LimitDegrees=45
  swing1Motion=Locked
  swing2LimitDegrees=45
  swing2Motion=Locked
  twistLimitDegrees=45
  twistMotion=Locked
 }
 {bone1Name=dyn_jacket_l, bone2Name=base_jacket_l, swing1LimitDegrees=10, swing2LimitDegrees=20, twistLimitDegrees=10}
 {
  bone1Name=base_jacket_r
  bone2Name=spine_03
  swing1LimitDegrees=45
  swing1Motion=Locked
  swing2LimitDegrees=45
  swing2Motion=Locked
  twistLimitDegrees=45
  twistMotion=Locked
 }
 {bone1Name=dyn_jacket_r, bone2Name=base_jacket_r, swing1LimitDegrees=10, swing2LimitDegrees=20, twistLimitDegrees=10}
 {
  bone1Name=dyn_jacket_bck_l
  bone2Name=dyn_jacket_bck
  swing1LimitDegrees=45
  swing1Motion=Free
  swing2LimitDegrees=45
  swing2Motion=Free
  twistLimitDegrees=45
  twistMotion=Free
 }
 {
  bone1Name=dyn_jacket_bck_r
  bone2Name=dyn_jacket_bck
  swing1LimitDegrees=45
  swing1Motion=Free
  swing2LimitDegrees=45
  swing2Motion=Free
  twistLimitDegrees=45
  twistMotion=Free
 }
]
