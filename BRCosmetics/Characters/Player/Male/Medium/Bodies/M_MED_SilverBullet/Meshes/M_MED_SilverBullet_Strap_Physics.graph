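# Physics asset for the left hip strap on the medium male SilverBullet player cosmetic, set up for strap dynamics against the thigh.
# A small base capsule is angularly locked to a large left-thigh capsule and anchors a chain of three dynamic strap capsules.
# Each chain joint carries identical limit angles, with a wider swing on one axis than the other and zero twist.
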PhysicsAsset /BRCosmetics/Characters/Player/Male/Medium/Bodies/M_MED_SilverBullet/Meshes/M_MED_SilverBullet_Strap_Physics
bodies {
 base_hipStrap_01_l [{shapeType=Capsule, length=0.505, radius=0.505, shapeName=base_hipStrap_01_l_capsule, rotation=<rot 0 -90 0>}]
 dyn_hipStrap_01_l [{shapeType=Capsule, length=4.2392, radius=2, shapeName=dyn_hipStrap_01_l_capsule, center=<3.5841 0 0.6944>, rotation=<rot 90 0 0>}]
 dyn_hipStrap_02_l [{shapeType=Capsule, length=1, radius=2, shapeName=dyn_hipStrap_02_l_capsule, center=<1.2296 0 1.2785>, rotation=<rot 90 0 0>}]
 dyn_hipStrap_03_l [{shapeType=Capsule, length=1, radius=2, shapeName=dyn_hipStrap_03_l_capsule, center=<1.7773 0 1.2756>, rotation=<rot 90 0 0>}]
 thigh_l [{shapeType=Capsule, length=24.2606, radius=8.7858, center=<-16.2994 -0.2812 0>, rotation=<rot 90 0 0>}]
}
constraints [
 {
  bone1Name=base_hipStrap_01_l
  bone2Name=thigh_l
  swing1LimitDegrees=45
  swing1Motion=Locked
  swing2LimitDegrees=45
  swing2Motion=Locked
  twistLimitDegrees=45
  twistMotion=Locked
 }
 {bone1Name=dyn_hipStrap_01_l, bone2Name=base_hipStrap_01_l, swing1LimitDegrees=35, swing2LimitDegrees=15, twistLimitDegrees=0}
 {bone1Name=dyn_hipStrap_02_l, bone2Name=dyn_hipStrap_01_l, swing1LimitDegrees=35, swing2LimitDegrees=15, twistLimitDegrees=0}
 {bone1Name=dyn_hipStrap_03_l, bone2Name=dyn_hipStrap_02_l, swing1LimitDegrees=35, swing2LimitDegrees=15, twistLimitDegrees=0}
]
