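# Post-process style Animation Blueprint for the SilverBullet medium male body, adding secondary motion to the pose copied from its parent mesh.
# A Control Rig receives skydiving and DBNO state, with the skydive bool converted to float each update.
# Its cached output feeds RigidBody jacket and strap simulations, then spring-driven AnimDynamics shoelaces.
# Fixed additive bone offsets adjust the three back jacket bones after simulation, before output.
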AnimBlueprint /BRCosmetics/Characters/Player/Male/Medium/Bodies/M_MED_SilverBullet/Meshes/M_MED_SilverBullet_AnimBP
target_skeleton=&/M_MED_SilverBullet_Skeleton
graphs [
 AnimGraph {nodes [
   Copy Pose From Mesh() {bUseAttachedParent=true}
   n1 = "Transform (Modify) Bone - Bone: dyn_jacket_bck"!(
    ComponentPose=Local To Component(LocalPose=Use cached pose 'CRCache_dynamics'())
    bAlphaBoolEnabled:bool=True
    Translation="2,0,0"
    Rotation="-15,0,0"
    Scale="1,1,1"
   ) -> (Pose) {
    alphaInputType=Bool
    rotationMode=BMM_Additive
    rotationSpace=BCS_BoneSpace
    translationMode=BMM_Additive
    translationSpace=BCS_BoneSpace
    alphaBoolBlend {blendInTime=0.2, blendOption=Linear, blendOutTime=0.2}
    boneToModify {boneName=dyn_jacket_bck}
    rotation {pitch=-15}
    translation {x=2}
   }
   n2 = "Transform (Modify) Bone - Bone: dyn_jacket_bck_l"!(ComponentPose=@n1, bAlphaBoolEnabled:bool=True, Translation="0,0,0", Rotation="-15,-10,0", Scale="1,1,1") -> (Pose) {
    alphaInputType=Bool
    rotationMode=BMM_Additive
    rotationSpace=BCS_BoneSpace
    translationMode=BMM_Additive
    translationSpace=BCS_BoneSpace
    alphaBoolBlend {blendInTime=0.2, blendOption=Linear, blendOutTime=0.2}
    boneToModify {boneName=dyn_jacket_bck_l}
    rotation {pitch=-15, yaw=-10}
   }
   n3 = "Transform (Modify) Bone - Bone: dyn_jacket_bck_r"!(ComponentPose=@n2, bAlphaBoolEnabled:bool=True, Translation="0,0,0", Rotation="-15,-10,0", Scale="1,1,1") -> (Pose) {
    alphaInputType=Bool
    rotationMode=BMM_Additive
    rotationSpace=BCS_BoneSpace
    translationMode=BMM_Additive
    translationSpace=BCS_BoneSpace
    alphaBoolBlend {blendInTime=0.2, blendOption=Linear, blendOutTime=0.2}
    boneToModify {boneName=dyn_jacket_bck_r}
    rotation {pitch=-15, yaw=-10}
   }
   Output Pose(Result=Component To Local(ComponentPose=@n3)) {layerGroup=None}
   n4 = Control Rig!(
    Source=Copy Pose From Mesh() {bUseAttachedParent=true}
    Alpha
    SkyDive:double=Get skydiveFloat() -> (skydiveFloat:double)
    isDBNO:bool=Get bIsDBNO()
   ) -> (Pose) {
    alphaBoolBlend {blendOption=Linear}
    controlRigAssetReference {blueprintRigClass=&/M_MED_SilverBullet_CtrlRig.M_MED_SilverBullet_CtrlRig_C}
   }
   Save cached pose 'CRCache_Input'(Pose=@n4) {cacheName=CRCache_Input}
   n5 = RigidBody(ComponentPose=Local To Component(LocalPose=Use cached pose 'CRCache_Input'())) {
    bOverrideWorldGravity=true
    overridePhysicsAsset=&/M_MED_SilverBullet_Jacket_Physics
    simulationSpace=BaseBoneSpace
    alphaBoolBlend {blendOption=Linear}
    baseBoneRef {boneName=spine_03}
    componentAppliedLinearAccClamp=<3000 3000 3000>
    overrideWorldGravity {z=-980}
    simSpaceSettings {
     maxAngularAcceleration=1
     maxAngularVelocity=3
     maxLinearAcceleration=1000
     maxLinearVelocity=200
     worldAlpha=1
     externalLinearDragV=<0.2 0.2 0.2>
    }
   }
   n6 = RigidBody(ComponentPose=@n5) {
    bOverrideWorldGravity=true
    overridePhysicsAsset=&/M_MED_SilverBullet_Strap_Physics
    simulationSpace=BaseBoneSpace
    alphaBoolBlend {blendOption=Linear}
    baseBoneRef {boneName=thigh_l}
    componentAppliedLinearAccClamp=<3000 3000 3000>
    overrideWorldGravity {z=-980}
    simSpaceSettings {
     maxAngularAcceleration=1
     maxAngularVelocity=3
     maxLinearAcceleration=1000
     maxLinearVelocity=500
     worldAlpha=1
     externalLinearDragV=<0.2 0.2 0.2>
    }
   }
   n7 = Anim Dynamics - Bone: dyn_shoelace_r(ComponentPose=@n6, Alpha=1, ExternalForce="0,0,0") {
    angularDampingOverride=1
    angularSpringConstant=20000
    bAngularSpring=true
    bOverrideAngularDamping=true
    bUseGravityOverride=true
    gravityScale=0.35
    alphaBoolBlend {blendOption=Linear}
    boundBone {boneName=dyn_shoelace_r}
    componentLinearAccScale=<1 1 1>
    componentLinearVelScale=<1 1 1>
    gravityOverride {z=-980}
    physicsBodyDefinitions [{
      boundBone {boneName=dyn_shoelace_r}
      boxExtents=<20 30 30>
      constraintSetup {angularLimitsMax=<30 1 0>, angularLimitsMin=<-30 -80 0>, angularTarget=<1 0 0>}
      localJointOffset=<0 0 4>
     }]
   }
   n8 = Anim Dynamics - Bone: dyn_shoelace_l(ComponentPose=@n7, Alpha=1, ExternalForce="0,0,0") {
    angularDampingOverride=1
    angularSpringConstant=20000
    bAngularSpring=true
    bOverrideAngularDamping=true
    bUseGravityOverride=true
    gravityScale=0.35
    alphaBoolBlend {blendOption=Linear}
    boundBone {boneName=dyn_shoelace_l}
    componentLinearAccScale=<1 1 1>
    componentLinearVelScale=<1 1 1>
    gravityOverride {z=-980}
    physicsBodyDefinitions [{
      boundBone {boneName=dyn_shoelace_l}
      boxExtents=<20 30 30>
      constraintSetup {angularLimitsMax=<30 1 0>, angularLimitsMin=<-30 -80 0>, angularTarget=<1 0 0>}
      localJointOffset=<0 0 4>
     }]
   }
   Save cached pose 'CRCache_dynamics'(Pose=Component To Local(ComponentPose=@n8)) {cacheName=CRCache_dynamics}
  ]}
 EventGraph {nodes [
   Event BlueprintUpdateAnimation()
   > Set skydiveFloat(skydiveFloat:double="To Float (Boolean)"(InBool=Get bIsSkyDiving())) -> (Output_Get:double)
  ]}
]
variables [
 isWearingGauntlets {type=boolean}
 skydiveFloat {type=number}
]
