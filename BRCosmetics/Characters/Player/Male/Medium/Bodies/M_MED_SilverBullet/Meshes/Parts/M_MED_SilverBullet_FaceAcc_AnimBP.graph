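# Animation Blueprint for the face accessory part of the male medium SilverBullet cosmetic character body.
# It copies the pose from the attached parent mesh, then applies a Transform (Modify) Bone node to the cap flap bone dyn_capflap_m.
# That node adds a fixed pitch and Z offset in bone space, with its Bool alpha set to true.
# The EventGraph holds only an unconnected update event, and no variables drive the pose.
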AnimBlueprint /BRCosmetics/Characters/Player/Male/Medium/Bodies/M_MED_SilverBullet/Meshes/Parts/M_MED_SilverBullet_FaceAcc_AnimBP
target_skeleton=&/M_MED_SilverBullet_FaceAcc_Skeleton
graphs [
 AnimGraph {nodes [
   n1 = "Transform (Modify) Bone - Bone: dyn_capflap_m"!(
    ComponentPose=Local To Component(LocalPose=Copy Pose From Mesh() {bUseAttachedParent=true})
    bAlphaBoolEnabled:bool=True
    Translation="0,0,3"
    Rotation="60,0,0"
    Scale="1,1,1"
   ) -> (Pose) {
    alphaInputType=Bool
    rotationMode=BMM_Additive
    rotationSpace=BCS_BoneSpace
    translationMode=BMM_Additive
    translationSpace=BCS_BoneSpace
    alphaBoolBlend {blendInTime=0.2, blendOption=Linear, blendOutTime=0.5}
    boneToModify {boneName=dyn_capflap_m}
    rotation {pitch=60}
    translation {z=3}
   }
   Output Pose(Result=Component To Local(ComponentPose=@n1)) {layerGroup=None}
  ]}
 EventGraph {nodes [Event BlueprintUpdateAnimation()]}
]
variables []
